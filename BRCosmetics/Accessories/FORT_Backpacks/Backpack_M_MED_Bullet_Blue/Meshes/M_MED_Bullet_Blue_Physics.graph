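# Physics asset for the M_MED_Bullet_Blue backpack cosmetic, defining collision and joint limits for its dangling parts.
# A two-box pack body anchors the bullet strap, two capsule canisters and two box pouches through limited swing constraints, with twist locked on the pouches.
# Two small strap bodies hang from a separate dyn_pack box, which has no constraint to the main pack body.
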PhysicsAsset /BRCosmetics/Accessories/FORT_Backpacks/Backpack_M_MED_Bullet_Blue/Meshes/M_MED_Bullet_Blue_Physics
bodies {
 dyn_bullet_strap [{shapeType=Box, extentX=7.3905, extentY=3.3693, extentZ=14.4111, center=<3.5 3.5 1>, rotation=<rot 0 10.5291 -87.2242>}]
 dyn_cannister_l [{shapeType=Capsule, length=5.5364, radius=2.8573, center=<6.5 -1.5 -0.3>, rotation=<rot 90 0 -3.1884>}]
 dyn_cannister_r [{shapeType=Capsule, length=4.8851, radius=2.724, center=<7 -1.1 -0.6>, rotation=<rot 87.9193 0.1111 0.0009>}]
 dyn_pack [{shapeType=Box, extentX=13.3965, extentY=7.5, extentZ=15.3817, center=<5.9029 -0.0001 0>}]
 dyn_pouch_l [{shapeType=Box, extentX=10, extentY=1, extentZ=6, center=<5 0 0>, rotation=<rot 90 0 0>}]
 dyn_pouch_r [{shapeType=Box, extentX=10, extentY=1, extentZ=6, center=<5 0 0>, rotation=<rot 90 0 0>}]
 dyn_strap_a [{shapeType=Box, extentX=2, extentY=2, extentZ=1, center=<0.5 0 0>}]
 dyn_strap_b [{shapeType=Box, extentX=2, extentY=2, extentZ=1, center=<0.5 0 0>}]
 pack [
  {shapeType=Box, extentX=26.8963, extentY=10.1549, extentZ=38.2337, center=<-0.2287 -6.0361 -7.1095>}
  {shapeType=Box, extentX=20.235, extentY=7.8178, extentZ=12.4114, center=<-1.8024 -10.3017 -0.0826>}
 ]
}
constraints [
 {bone1Name=dyn_bullet_strap, bone2Name=pack, swing1LimitDegrees=20, swing2LimitDegrees=15, twistLimitDegrees=5}
 {bone1Name=dyn_pouch_r, bone2Name=pack, swing1LimitDegrees=30, swing2LimitDegrees=5, twistLimitDegrees=0, twistMotion=Locked}
 {bone1Name=dyn_pouch_l, bone2Name=pack, swing1LimitDegrees=30, swing2LimitDegrees=5, twistLimitDegrees=5, twistMotion=Locked}
 {bone1Name=dyn_cannister_l, bone2Name=pack, swing1LimitDegrees=45, swing2LimitDegrees=45, twistLimitDegrees=25}
 {bone1Name=dyn_cannister_r, bone2Name=pack, swing1LimitDegrees=45, swing2LimitDegrees=45, twistLimitDegrees=25}
 {bone1Name=dyn_strap_b, bone2Name=dyn_pack, swing1LimitDegrees=5, swing2LimitDegrees=25, twistLimitDegrees=5}
 {bone1Name=dyn_strap_a, bone2Name=dyn_pack, swing1LimitDegrees=5, swing2LimitDegrees=25, twistLimitDegrees=5}
]
